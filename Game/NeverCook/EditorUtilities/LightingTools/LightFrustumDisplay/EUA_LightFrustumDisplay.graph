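# Editor utility Blueprint for a lighting tool that, by its name, displays light frustums, with each instance linked to one Local Light component.
# SetComponent stores the light and writes its CastShadows flag as a float into the Mesh's first custom primitive data slot, both current and default.
# SelectActor uses EditorActorSubsystem to select the stored light's owning actor in the level, if the light is valid.
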EditorUtilityBlueprint /Game/NeverCook/EditorUtilities/LightingTools/LightFrustumDisplay/EUA_LightFrustumDisplay
graphs [
 EventGraph {nodes [
   Event BeginPlay()
   Event ActorBeginOverlap()
   Event Tick!() -> (OutputDelegate, then, DeltaSeconds:float)
  ]}
 SelectActor {nodes [
   SelectActor() {is=title K2Node_FunctionEntry}
   > Is Valid(InputObject=Get Local Light() -> (Local Light:Local Light Component Object Reference))
   n1 = Get Local Light()
   > SetSelectedLevelActors(
    self=Get EditorActorSubsystem()
    ActorsToSelect=Make Array("[0]":Actor Object Reference=GetOwner(self=@n1)) -> (Array:Array of Actor Object References)
   )
  ]}
 SetComponent {nodes [
   n1 = SetComponent!() -> (then, LocalLight:Local Light Component Object Reference) {is=title K2Node_FunctionEntry}
   > Set Local Light(Local Light:Local Light Component Object Reference=@n1) -> (Output_Get:Local Light Component Object Reference)
   n2 = "To Float (Boolean)"(InBool=Get CastShadows!(self:Light Component Base Object Reference=Get Local Light()) -> (CastShadows:bool))
   n3 = Get Mesh!() -> (Mesh:Static Mesh Component Object Reference)
   > SetCustomPrimitiveDataFloat!(execute, self:Primitive Component Object Reference=@n3, DataIndex:int=0, Value:float=@n2) -> (then)
   > SetDefaultCustomPrimitiveDataFloat!(execute, self:Primitive Component Object Reference=@n3, DataIndex:int=0, Value:float=@n2) -> (then)
  ]}
 UserConstructionScript {nodes [Construction Script()]}
]
variables [Local Light {type=LocalLightComponent, default=null}]
