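# EUA_LightFrustumDisplay_Point is an Editor Utility Blueprint in the lighting tools folder, named for displaying point light frustums.
# It holds no logic or variables of its own.
# Its event graph has only unconnected BeginPlay, ActorBeginOverlap and Tick events, and its construction script is empty.
# The file does not show any display behavior or the parent class.
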EditorUtilityBlueprint /Game/NeverCook/EditorUtilities/LightingTools/LightFrustumDisplay/EUA_LightFrustumDisplay_Point
graphs [
 EventGraph {nodes [
   Event BeginPlay()
   Event ActorBeginOverlap()
   Event Tick!() -> (OutputDelegate, then, DeltaSeconds:float)
  ]}
 UserConstructionScript {nodes [Construction Script()]}
]
variables []
